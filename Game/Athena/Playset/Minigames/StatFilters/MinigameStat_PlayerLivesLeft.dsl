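; MinigameStat_PlayerLivesLeft is a minigame stat filter, named for tracking how many lives a player has left in playset minigames.
; Its Compare function takes two integer scores and returns the first minus the second, so a positive result means the first score is higher.
; The file holds only this comparison and shows no logic for reading or updating lives.
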(fn Compare
  (params (FirstScore :type Integer) (SecondScore :type Integer))
  (return (- FirstScore SecondScore)))
