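; Minigame stat filter for times in milliseconds, where a lower valid time is better.
; Accumulate keeps the smaller valid time, falling back to whichever is valid.
; Compare returns second minus first when validity matches and otherwise favors the valid time, with a debug print as fallback.
; FormatStat converts through Timespan to show seconds and milliseconds, adding minutes and hours by duration, or dashes when invalid.
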(fn Accumulate
  (params (A :type Integer) (B :type Integer))
  (return (select (and (Utilities|IsTimeValid A) (Utilities|IsTimeValid B)) (Math|Integer|Min(Integer) A B) (select (Utilities|IsTimeValid B) B A))))

(fn Compare
  (params (FirstScore :type Integer) (SecondScore :type Integer))
  (Utilities|FlowControl|Sequence
    (:then_0
      (LocalVariables|SetFirstIsValid (Utilities|IsTimeValid (Variables|GetFirstScore)))
      (LocalVariables|SetSecondIsValid (Utilities|IsTimeValid (Variables|GetSecondScore))))
    (:then_1
      (if (Math|Boolean|Equal(Boolean) (LocalVariables|GetFirstIsValid) (LocalVariables|GetSecondIsValid))
        (return (- (Variables|GetSecondScore) (Variables|GetFirstScore)))))
    (:then_2
      (if (and (LocalVariables|GetFirstIsValid) (not (LocalVariables|GetSecondIsValid)))
        (return 1)))
    (:then_3
      (if (and (not (LocalVariables|GetFirstIsValid)) (LocalVariables|GetSecondIsValid))
        (return -1)))
    (:then_4
      (Development|PrintString 0 "MinigameStat_Time has broken Compare function")
      (return 0))))

(fn FormatStat
  (params (InCount :type Integer))
  (return (select (Utilities|IsTimeValid InCount) (select (or (> (Math|Timespan|GetTotalHours (Math|Timespan|FromMilliseconds (Math|Conversions|ToFloat(Integer) InCount))) 1.0) (> (Math|Timespan|GetTotalMinutes (Math|Timespan|FromMilliseconds (Math|Conversions|ToFloat(Integer) InCount))) 1.0)) (select (> (Math|Timespan|GetTotalHours (Math|Timespan|FromMilliseconds (Math|Conversions|ToFloat(Integer) InCount))) 1.0) (Utilities|Text|FormatText "NSLOCTEXT(\"[733D8F6049320C080C273F8D51EB3018]\", \"51FE2EEC45ED4687B379E382D7E33771\", \"{Hours}:{Minutes}:{Seconds}.{Milleseconds}\")" (Utilities|Text|ToText(Integer) (Math|Timespan|GetHours (Math|Timespan|FromMilliseconds (Math|Conversions|ToFloat(Integer) InCount))) false false 2 2) (Utilities|Text|ToText(Integer) (Math|Timespan|GetMinutes (Math|Timespan|FromMilliseconds (Math|Conversions|ToFloat(Integer) InCount))) false false 2 2) (Utilities|Text|ToText(Integer) (Math|Timespan|GetSeconds (Math|Timespan|FromMilliseconds (Math|Conversions|ToFloat(Integer) InCount))) false false 2 2) (Utilities|Text|ToText(Integer) (Math|Timespan|GetMilliseconds (Math|Timespan|FromMilliseconds (Math|Conversions|ToFloat(Integer) InCount))) false false 3 3)) (Utilities|Text|FormatText "NSLOCTEXT(\"[733D8F6049320C080C273F8D51EB3018]\", \"F7C8DA464569C8EDCBE4E2ACC57FEBFB\", \"{Minutes}:{Seconds}.{Milleseconds}\")" (Utilities|Text|ToText(Integer) (Math|Timespan|GetMinutes (Math|Timespan|FromMilliseconds (Math|Conversions|ToFloat(Integer) InCount))) false false 2 2) (Utilities|Text|ToText(Integer) (Math|Timespan|GetSeconds (Math|Timespan|FromMilliseconds (Math|Conversions|ToFloat(Integer) InCount))) false false 2 2) (Utilities|Text|ToText(Integer) (Math|Timespan|GetMilliseconds (Math|Timespan|FromMilliseconds (Math|Conversions|ToFloat(Integer) InCount))) false false 3 3))) (Utilities|Text|FormatText "NSLOCTEXT(\"[733D8F6049320C080C273F8D51EB3018]\", \"20E2A839438F005CC160F78F6CAAD96D\", \"{Seconds}.{Milleseconds}\")" (Utilities|Text|ToText(Integer) (Math|Timespan|GetSeconds (Math|Timespan|FromMilliseconds (Math|Conversions|ToFloat(Integer) InCount))) false false 2 2) (Utilities|Text|ToText(Integer) (Math|Timespan|GetMilliseconds (Math|Timespan|FromMilliseconds (Math|Conversions|ToFloat(Integer) InCount))) false false 3 3))) "NSLOCTEXT(\"[733D8F6049320C080C273F8D51EB3018]\", \"3E0A696640F12BAD7635ADB0F5A9ECB1\", \"--:--:--\")")))
