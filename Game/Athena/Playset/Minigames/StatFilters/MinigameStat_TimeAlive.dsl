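; Minigame stat filter for time alive in Athena playset minigames, defining how scores accumulate and rank.
; Accumulate safely adds two positive times, otherwise keeps B if positive, else A.
; Compare overrides the parent's stopwatch ordering, where smaller is better, so longer times rank higher.
; When both scores share validity (above zero or not) it returns their difference; otherwise the valid one wins.
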(fn Accumulate
  (params (A :type Integer) (B :type Integer))
  (return (select (and (> A 0) (> B 0)) (Minigame|Stats|AddSafe A B) (select (> B 0) B A))))

(fn Compare
  (params (FirstScore :type Integer) (SecondScore :type Integer))
  ;[1 1 1] Parent class uses the stopwatch model where the smaller time is greater than the larger. This class should compare times normally.
  (Utilities|FlowControl|Sequence
    (:then_0
      (LocalVariables|SetIsFirstValid (> (Variables|GetFirstScore) 0))
      (LocalVariables|SetIsSecondValid (> (Variables|GetSecondScore) 0)))
    (:then_1
      (if (Math|Boolean|Equal(Boolean) (LocalVariables|GetIsFirstValid) (LocalVariables|GetIsSecondValid))
        (return (- (Variables|GetFirstScore) (Variables|GetSecondScore)))))
    (:then_2
      (if (and (LocalVariables|GetIsFirstValid) (not (LocalVariables|GetIsSecondValid)))
        (return 1)))
    (:then_3
      (if (and (not (LocalVariables|GetIsFirstValid)) (LocalVariables|GetIsSecondValid))
        (return -1)))
    (:then_4
      (return 0))))
